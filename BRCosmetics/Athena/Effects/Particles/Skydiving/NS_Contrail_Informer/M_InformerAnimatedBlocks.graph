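# Translucent, unlit, two-sided material for Niagara mesh particles, producing animated blocks for the Informer skydiving contrail.
# A looping time ramp builds an offset along vertex normals, with distance set by the Spreed_Speed dynamic parameter.
# Particle color is tinted by Shade_Color according to face side and fed through emissive adjustments.
# Particle alpha multiplies a procedural square edge mask, whose center-scaled UVs follow Shrink_Speed.
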Material /BRCosmetics/Athena/Effects/Particles/Skydiving/NS_Contrail_Informer/M_InformerAnimatedBlocks
graph {
 properties {
  bUsedWithNiagaraMeshParticles=true
  blendMode=BLEND_Translucent
  pixelDepthOffsetMode=PDOM_Legacy
  previewMesh=&/SM_InformerANimatedBlocks
  shadingModel=MSM_Unlit
  twoSided=true
  usedShadingModels=Unlit
 }
 nodes [
  n1 = Constant3Vector() {constant=<rgba 100 100 100 0>}
  n2 = VertexNormalWS()
  n3 = DynamicParameter() {defaultValue=<rgba 1 0 1 0>, paramNames [Spreed_Speed, Shrink_Speed, "", ""]}
  n4 = Time()
  LinearInterpolate(
   A=Constant3Vector()
   B=Multiply(A=Multiply(A=@n1, B=@n2) {constA=1}, B=@n3.Spreed_Speed)
   Alpha=Fmod(A=Multiply(A=@n4) {constB=0.1}, B=Constant() {r=1})
  )
  n5 = ParticleColor()
  n6 = VectorParameter() {parameterName=Shade_Color, defaultValue=<rgba 0.2903 0.4145 0.6406 1>}
  MaterialFunctionCall(Emissive=LinearInterpolate(A=@n5.RGB, B=Multiply(A=@n6.RGB, B=@n5.RGB), Alpha=TwoSidedSign())) {materialFunction=&MF_FX_EmissiveAdjustments}
  n7 = TextureCoordinate()
  n8 = Constant2Vector() {g=1, r=1}
  n9 = LinearInterpolate(A=@n8, B=Multiply(A=@n8) {constB=0.5}, Alpha=@n3.Shrink_Speed)
  Multiply(
   A=@n5.A
   B=MaterialFunctionCall!(
    UVs=MaterialFunctionCall(UVs=@n7, Texture Scale=@n9) {materialFunction=&ScaleUVsByCenter}.UVs
    Mult=Constant() {r=1}
    Exp=Constant() {r=0.25}
   ) {materialFunction=/Game/Effects/Fort_Effects/MaterialFunctions/MF_SquareEdgeMask}.SquareMask
  )
 ]
}
