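# Niagara skydiving contrail cosmetic "Informer" that trails green-to-orange-red effects from two skeleton bones of the user-supplied skeletal mesh.
# Eight emitters layer a wide ribbon, rising cylinders, squares and pattern sprites with a blue shadow color, plus bone-following decal sprites, lights, rays and fire meshes.
# A diving attribute stretches the fire meshes and enlarges the rays.
# Ribbon velocity differs between gameplay and front end.
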NiagaraSystem /BRCosmetics/Athena/Effects/Particles/Skydiving/NS_Contrail_Informer/NS_Contrail_Informer
dependencies {
 usedDataInterfaces [&NiagaraDataInterfaceCurve, &NiagaraDataInterfaceColorCurve, &NiagaraDataInterfaceVectorCurve]
 usedDynamicInputs []
 usedModules [&SystemState, /Niagara/Modules/System/SystemLifeCycle]
}
emitters {
 FN_COntrails_Cylinders {
  emitterSpawnScript [&FN_PreserveSkydiveUserVariables() {module=PreserveUserVariables}]
  emitterUpdateScript [
   &EmitterState()
   /Game/Effects/Niagara/Modules/Spawn/SpawnRateFromMultipleSpawnGroups_2(Global Spawnrate=6)
  ]
  particleSpawnScript [
   /Game/Effects/Niagara/Modules/Groups/FN_SetSpawnGroup_v1/FN_SetSpawnGroup(Emitter Bone Count=2)
   /Niagara/Modules/Spawn/Initialization/InitializeParticle(
    Color=<rgba 1.5 1.5 1.5 1>
    Module.Write Color=true
    Module.Write Lifetime=true
    Module.Write Mass=true
    Module.Write Material Random=true
    Module.Write Position=false
    Module.Write Scale=true
    Module.Write SpriteRotation=false
    Module.Write SpriteSize=false
    Position=&SimulationPosition()
    Sprite Size=<1 1>
   )
   &SampleSkeletalMeshSkeleton(Mesh:NiagaraDataInterfaceSkeletalMesh=$User.SkeletalMesh, Specific Bone Index:NiagaraInt32=$Particles.SpawnGroup)
   /Niagara/Modules/Spawn/Location/SkeletalMeshSkeletonLocation(Module.Inherit Velocity=false, Module.Use Sampled Bone as Ribbon ID=true, Set Mesh Orientation to Bone Orientation=false)
   &AddVelocity(Velocity=<0 0 250>)
   /Niagara/Modules/Spawn/Orientation/V2/InitialMeshOrientation(
    Orientation Axis:Vector3f=<0 0 1>
    Orientation Coordinate Space:ENiagaraCoordinateSpace=::World
    Orientation Vector=<0 0 1>
    Rotation=&MakeVector(X=0, Y=0, Z=&UniformRangedFloat(Minimum=0, Maximum=360))
    Use Orientation Vector=true
    Use Rotation Vector=true
   )
  ]
  particleUpdateScript [
   &ParticleState()
   &SolveForcesAndVelocity(Clamp Velocity=false, Limit Acceleration=false)
   &Color(Color=&ColorFromCurve(ColorCurve=DataInterface() {propertyValues {
       AlphaCurve {keys [<at -0.0381 0>, {arriveTangent=-0.0029, interpMode=RCIM_Cubic, leaveTangent=-0.0029, time=0.1237, value=1}, <at 0.9977 -0.003 RCIM_Cubic>]}
       BlueCurve {keys [<at -0.0023 0 RCIM_Cubic>, {arriveTangent=0.1169, interpMode=RCIM_Cubic, leaveTangent=0.1169, time=0.6427, value=0}, <at 1.017 0.1192>]}
       GreenCurve {keys [
         <at -0.0023 1 RCIM_Cubic>
         {arriveTangent=-0.8537, interpMode=RCIM_Cubic, leaveTangent=-0.8537, time=0.6427, value=0.0988}
         <at 1.017 0.1298>
        ]}
       RedCurve {keys [<at -0.0023 0.053 RCIM_Cubic>, {arriveTangent=0.8116, interpMode=RCIM_Cubic, leaveTangent=0.8116, time=0.6427, value=1}, <at 1.017 0.8802>]}
      }}))
   &ScaleMeshSize(Scale Factor=&VectorFromFloat(Value=&FloatFromCurve(
      FloatCurve=DataInterface() {propertyValues {Curve {keys [
          {arriveTangent=2.4656, interpMode=RCIM_Cubic, leaveTangent=2.4656, tangentMode=RCTM_User, time=-0.0046, value=0.2}
          {arriveTangent=0.2615, interpMode=RCIM_Cubic, leaveTangent=0.2615, tangentMode=RCTM_User, time=1.0011, value=1}
         ]}}}
      Scale Curve=0.3
     )))
   &DynamicMaterialParameters(
    Index 0 Param 1=&FloatFromCurve(FloatCurve=DataInterface() {propertyValues {Curve {keys [<at -0.0046 0>,<at 1.0011 1>]}}}, Scale Curve=0.5)
    Index 0 Param 2=&FloatFromCurve(FloatCurve=DataInterface() {propertyValues {Curve {keys [
         {interpMode=RCIM_Cubic, tangentMode=RCTM_User, time=-0.0046, value=0}
         {arriveTangent=1.1308, interpMode=RCIM_Cubic, leaveTangent=1.1308, tangentMode=RCTM_User, time=1.0011, value=1}
        ]}}})
    Module.Param0WriteEnabled=true
    Module.Param1WriteEnabled=false
    Module.Param2WriteEnabled=false
    Module.Param3WriteEnabled=false
   )
  ]
  rendererClasses [&NiagaraMeshRendererProperties]
  renderers [{rendererClass=&NiagaraMeshRendererProperties, rendererIndex=0}]
 }
 FN_COntrails_Squares {
  emitterSpawnScript [&FN_PreserveSkydiveUserVariables() {module=PreserveUserVariables}]
  emitterUpdateScript [
   &EmitterState()
   &SpawnRateFromMultipleSpawnGroups_2(Global Spawnrate=8)
  ]
  particleSpawnScript [
   &FN_SetSpawnGroup(Emitter Bone Count=2)
   &InitializeParticle(
    Color=<rgba 1.5 1.5 1.5 1>
    Mesh Scale=<0.5 0.5 1>
    Module.Write Color=true
    Module.Write Lifetime=true
    Module.Write Mass=true
    Module.Write Material Random=true
    Module.Write Position=false
    Module.Write Scale=false
    Module.Write SpriteRotation=false
    Module.Write SpriteSize=true
    Position=&SimulationPosition()
    Sprite Size=&Vector2DFromFloat(Value=&UniformRangedFloat(Minimum=3, Maximum=7))
   )
   &SampleSkeletalMeshSkeleton(Mesh:NiagaraDataInterfaceSkeletalMesh=$User.SkeletalMesh, Specific Bone Index:NiagaraInt32=$Particles.SpawnGroup)
   &SkeletalMeshSkeletonLocation(Module.Inherit Velocity=false, Module.Use Sampled Bone as Ribbon ID=true, Set Mesh Orientation to Bone Orientation=false)
   &AddVelocity(Velocity=<0 0 250>)
   &InitialMeshOrientation(
    Orientation Axis:Vector3f=<0 0 1>
    Orientation Coordinate Space:ENiagaraCoordinateSpace=::World
    Orientation Vector=<0 0 1>
    Rotation=&MakeVector(X=0, Y=0, Z=&UniformRangedFloat(Minimum=0, Maximum=360))
    Use Orientation Vector=true
    Use Rotation Vector=true
   )
   /Niagara/Modules/Spawn/Location/SphereLocation(MaskSpawnGroup=false, Module.Override Randomness=false, Sphere Radius=8, Surface Only=false)
  ]
  particleUpdateScript [
   &ParticleState()
   &SolveForcesAndVelocity(Clamp Velocity=false, Limit Acceleration=false)
   &Color(Color=&ColorFromCurve(ColorCurve=DataInterface() {propertyValues {
       AlphaCurve {keys [<at -0.0381 0>, {arriveTangent=-0.0029, interpMode=RCIM_Cubic, leaveTangent=-0.0029, time=0.1237, value=1}, <at 0.9977 -0.003 RCIM_Cubic>]}
       BlueCurve {keys [<at -0.0023 0 RCIM_Cubic>, {arriveTangent=0.1169, interpMode=RCIM_Cubic, leaveTangent=0.1169, time=0.6427, value=0}, <at 1.017 0.1192>]}
       GreenCurve {keys [
         <at -0.0023 1 RCIM_Cubic>
         {arriveTangent=-0.8537, interpMode=RCIM_Cubic, leaveTangent=-0.8537, time=0.6427, value=0.0988}
         <at 1.017 0.1298>
        ]}
       RedCurve {keys [<at -0.0023 0.053 RCIM_Cubic>, {arriveTangent=0.8116, interpMode=RCIM_Cubic, leaveTangent=0.8116, time=0.6427, value=1}, <at 1.017 0.8802>]}
      }}))
   &ScaleSpriteSize(Scale Factor=&Vector2DFromFloat(Value=&FloatFromCurve(FloatCurve=DataInterface() {propertyValues {Curve {keys [<at -0.0046 0.3>,<at 1.0011 0.8>]}}}, Scale Curve=1.5)))
  ]
  rendererClasses [&NiagaraSpriteRendererProperties]
  renderers [{rendererClass=&NiagaraSpriteRendererProperties}]
 }
 FN_Contrails_Decal {
  emitterSpawnScript [&FN_PreserveSkydiveUserVariables() {module=PreserveUserVariables}]
  emitterUpdateScript [
   &EmitterState()
   &SpawnRateFromMultipleSpawnGroups_2(Global Spawnrate=2)
   &SpawnBurst_Instantaneous(Spawn Count=2, Use Spawn Probability=false) {enabled=false}
  ]
  particleSpawnScript [
   &FN_SetSpawnGroup(Emitter Bone Count=2)
   &InitializeParticle(
    Color=<rgba 0.5 1 0 0.5>
    Lifetime=0.7
    Module.Write Color=true
    Module.Write Lifetime=true
    Module.Write Mass=true
    Module.Write Material Random=true
    Module.Write Position=false
    Module.Write Scale=false
    Module.Write SpriteRotation=false
    Module.Write SpriteSize=true
    Position=&SimulationPosition()
    Sprite Size=&Vector2DFromFloat(Value=50)
   )
  ]
  particleUpdateScript [
   &ParticleState()
   &SampleSkeletalMeshSkeleton(Mesh:NiagaraDataInterfaceSkeletalMesh=$User.SkeletalMesh, Specific Bone Index:NiagaraInt32=$Particles.SpawnGroup)
   &SkeletalMeshSkeletonLocation(Module.Inherit Velocity=false, Module.Use Sampled Bone as Ribbon ID=true, Sampled Position Offset:Vector3f=<0 7 0>)
   &SolveForcesAndVelocity(Clamp Velocity=false, Limit Acceleration=false) {enabled=false}
   &ScaleColor(
    Scale Alpha=&FloatFromCurve(FloatCurve=DataInterface() {propertyValues {Curve {keys [<at 0.0001 0.3382>,<at 0.1393 0.6251>,<at 0.3202 0.5233>,<at 0.4613 0.8934>,<at 0.6241 0.3845>,<at 0.7724 0.6066>,<at 1.0006 0.3567>]}}})
    ScaleA=true
    ScaleRGB=false
   )
  ]
  rendererClasses [&NiagaraSpriteRendererProperties]
  renderers [{rendererClass=&NiagaraSpriteRendererProperties}]
 }
 FN_Contrails_Fire_Mesh {
  emitterSpawnScript [&FN_PreserveSkydiveUserVariables() {module=PreserveUserVariables}]
  emitterUpdateScript [
   &EmitterState(
    Enable Distance Culling=true
    Life Cycle Mode=::NewEnumerator1
    Loop Behavior=::NewEnumerator1
    Loop Duration Mode=::NewEnumerator1
    Module.bUseMaxDistance=true
    Module.bUseMinDistance=false
    Reset Age On Awaken=true
    Scalability Mode=::NewEnumerator1
    UseLoopDelay=false
   )
   &SpawnRateFromMultipleSpawnGroups_2(Global Spawnrate=3)
  ]
  particleSpawnScript [
   /Game/Effects/Niagara/Modules/Skydiving/FN_DivingAttributes/FN_DivingAttributes(Bone Index=2) {enabled=false}
   /Game/Effects/Niagara/Modules/Skydiving/FN_DivingAttributes/FN_DivingAttributes_v2(Bone Index=2, SkeletalMesh:NiagaraDataInterfaceSkeletalMesh=$User.SkeletalMesh)
   &FN_SetSpawnGroup(Emitter Bone Count=2) {enabled=false}
   &FN_SetSpawnGroup_v2(Number of Spawn Groups=2)
   &InitializeParticle(
    Lifetime=0.5
    Mesh Scale=<0.9 0.9 0.9>
    Module.Write Color=true
    Module.Write Lifetime=true
    Module.Write Mass=true
    Module.Write Material Random=true
    Module.Write Position=false
    Module.Write Scale=true
    Module.Write SpriteRotation=false
    Module.Write SpriteSize=true
    Position=&SimulationPosition()
    Sprite Size=<50 50>
   )
   &InitialMeshOrientation(
    Orientation Axis:Vector3f=<0 0 1>
    Orientation Coordinate Space:ENiagaraCoordinateSpace=::World
    Orientation Vector=<0 0 1>
    Rotation=&MakeVector(X=0, Y=0, Z=&UniformRangedFloat(Minimum=0, Maximum=1))
    Rotation Coordinate Space:ENiagaraMeshTransforms=::NewEnumerator1
    Use Orientation Vector=true
    Use Rotation Vector=true
   )
  ]
  particleUpdateScript [
   &ParticleState()
   &FN_DivingAttributes() {enabled=false, module=FN_DivingAttributes001}
   &FN_DivingAttributes_v2(SkeletalMesh:NiagaraDataInterfaceSkeletalMesh=$User.SkeletalMesh) {module=FN_DivingAttributes_v2001}
   &SampleSkeletalMeshSkeleton(
    Mesh:NiagaraDataInterfaceSkeletalMesh=$User.SkeletalMesh
    Sampled Rotation Coordinate Space:ENiagaraCoordinateSpace=::World
    Specific Bone Index:NiagaraInt32=$Particles.SpawnGroup
   )
   &SkeletalMeshSkeletonLocation(Module.Inherit Velocity=false, Module.Use Sampled Bone as Ribbon ID=true)
   /Niagara/Modules/Update/Orientation/V2/OrientMeshToVector(
    Override Incoming Orientation=1
    Override Quat=true
    Reference Axis Vector:Vector3f=<0 0 1>
    UseOptionalSideAxis=false
    UseOptionalUpAxis=false
   )
   &ScaleMeshSize(Scale Factor=&Lerp_Vector(A=<1 1 1>, B=<0.5 0.5 15>, Alpha=$Particles.Dive))
   &ScaleColor(
    Scale Alpha=&FloatFromCurve(FloatCurve=DataInterface() {propertyValues {Curve {keys [<at -0.0046 0>,<at 0.3 1>,<at 1.0011 0.002>]}}})
    ScaleA=true
    ScaleRGB=false
   )
  ]
  rendererClasses [&NiagaraMeshRendererProperties]
  renderers [{rendererClass=&NiagaraMeshRendererProperties, rendererIndex=0}]
 }
 FN_Contrails_InformerPatterns_WithShade {
  emitterSpawnScript [&FN_PreserveSkydiveUserVariables() {module=PreserveUserVariables}]
  emitterUpdateScript [
   &EmitterState()
   &SpawnRateFromMultipleSpawnGroups_2(Global Spawnrate:NiagaraFloat=&UniformRangedFloat(Minimum=6, Maximum=12))
  ]
  particleSpawnScript [
   &FN_SetSpawnGroup(Emitter Bone Count=2)
   &InitializeParticle(
    Color=<rgba 3 3 3 1>
    Module.Write Color=true
    Module.Write Lifetime=true
    Module.Write Mass=true
    Module.Write Material Random=true
    Module.Write Position=false
    Module.Write Scale=false
    Module.Write SpriteRotation=false
    Module.Write SpriteSize=true
    Position=&SimulationPosition()
    Sprite Size=&Vector2DFromFloat(Value=&FloatFromCurve(
      FloatCurve=DataInterface() {propertyValues {Curve {keys [<at -0.0046 1.0113>,<at 0.1107 0.9979>,<at 0.1217 0.5746>,<at 1.0142 0.513>]}}}
      CurveIndex=&UniformRangedFloat(Minimum=0, Maximum=1)
      Scale Curve=20
     ))
   )
   /SetVariables(Particles.SubImageIndex=&UniformRangedFloat(Minimum=0, Maximum=5)) {isSetParameters=true}
   &SampleSkeletalMeshSkeleton(Mesh:NiagaraDataInterfaceSkeletalMesh=$User.SkeletalMesh, Specific Bone Index:NiagaraInt32=$Particles.SpawnGroup)
   &SkeletalMeshSkeletonLocation(Module.Inherit Velocity=false, Module.Use Sampled Bone as Ribbon ID=true)
   &AddVelocity(Velocity=<0 0 200>)
   &BoxLocation(Box Size=<10 10 0>, MaskSpawnGroup=false, Module.Override Seed=false, Rotate Box=false, Surface Only=false)
  ]
  particleUpdateScript [
   /SetVariables(Particles.ColorShadow:LinearColor=&ColorFromCurve(ColorCurve=DataInterface() {propertyValues {
       AlphaCurve {keys [<at 0.6018 1 RCIM_Cubic>,<at 0.9977 -0.003 RCIM_Cubic>]}
       BlueCurve {keys [<at -0.0023 0.6979 RCIM_Cubic>]}
       GreenCurve {keys [<at -0.0023 0.0788 RCIM_Cubic>]}
       RedCurve {keys [<at -0.0023 0 RCIM_Cubic>]}
      }})) {isSetParameters=true}
   &ParticleState()
   &SolveForcesAndVelocity(Clamp Velocity=false, Limit Acceleration=false)
   &ScaleColor(
    Scale Alpha=&FloatFromCurve(FloatCurve=DataInterface() {propertyValues {Curve {keys [<at -0.0046 1.0113>,<at 0.5278 0.9789>,<at 1.0011 0.002>]}}})
    ScaleA=true
    ScaleRGB=false
   )
   &ScaleSpriteSize(Scale Factor=&Vector2DFromFloat(Value=&FloatFromCurve(
      FloatCurve=DataInterface() {propertyValues {Curve {keys [{arriveTangent=3.5303, interpMode=RCIM_Cubic, leaveTangent=3.5303, tangentMode=RCTM_User, time=-0.0046, value=0.2}, <at 1.0011 1>]}}}
      Scale Curve=0.7
     )))
  ]
  rendererClasses [&NiagaraSpriteRendererProperties]
  renderers [
   {rendererClass=&NiagaraSpriteRendererProperties}
   {rendererClass=&NiagaraSpriteRendererProperties, rendererIndex=1}
  ]
 }
 FN_Contrails_Light {
  emitterSpawnScript [&FN_PreserveSkydiveUserVariables() {module=PreserveUserVariables}]
  emitterUpdateScript [
   &EmitterState()
   &SpawnRateFromMultipleSpawnGroups_2(Global Spawnrate=2)
   &SpawnBurst_Instantaneous(Spawn Count=4, Use Spawn Probability=false) {enabled=false}
  ]
  particleSpawnScript [
   &FN_SetSpawnGroup(Emitter Bone Count=2)
   &InitializeParticle(
    Lifetime=0.5
    Module.Write Color=true
    Module.Write Lifetime=true
    Module.Write Mass=true
    Module.Write Material Random=true
    Module.Write Position=false
    Module.Write Scale=false
    Module.Write SpriteRotation=false
    Module.Write SpriteSize=true
    Position=&SimulationPosition()
    Sprite Size=&Vector2DFromFloat(Value=60)
   )
  ]
  particleUpdateScript [
   &ParticleState()
   &SampleSkeletalMeshSkeleton(Mesh:NiagaraDataInterfaceSkeletalMesh=$User.SkeletalMesh, Specific Bone Index:NiagaraInt32=$Particles.SpawnGroup)
   &SkeletalMeshSkeletonLocation(Module.Inherit Velocity=false, Module.Use Sampled Bone as Ribbon ID=true, Sampled Position Offset:Vector3f=<0 0 5>)
   &SolveForcesAndVelocity(Clamp Velocity=false, Limit Acceleration=false) {enabled=false}
   &Color(Color=<rgba 0.4917 1 0 1>)
  ]
  rendererClasses [&NiagaraLightRendererProperties, &NiagaraSpriteRendererProperties]
  renderers [
   {rendererClass=&NiagaraLightRendererProperties, rendererIndex=0}
   {rendererClass=&NiagaraSpriteRendererProperties, rendererIndex=1}
  ]
 }
 FN_Contrails_LightRays {
  emitterSpawnScript [&FN_PreserveSkydiveUserVariables() {module=PreserveUserVariables}]
  emitterUpdateScript [
   &EmitterState()
   &SpawnRateFromMultipleSpawnGroups_2(Global Spawnrate=12)
   &SpawnBurst_Instantaneous(Spawn Count=4, Use Spawn Probability=false) {enabled=false}
  ]
  particleSpawnScript [
   &FN_DivingAttributes(Bone Index=2)
   &FN_SetSpawnGroup(Emitter Bone Count=2)
   &InitializeParticle(
    Color=<rgba 0.3115 1 0 1>
    Lifetime=0.7
    Module.Write Color=true
    Module.Write Lifetime=true
    Module.Write Mass=true
    Module.Write Material Random=true
    Module.Write Position=false
    Module.Write Scale=false
    Module.Write SpriteRotation=false
    Module.Write SpriteSize=true
    Position=&SimulationPosition()
    Sprite Rotation=&UniformRangedFloat(Minimum=-15, Maximum=15)
    Sprite Size=<25 50>
   )
   /SetVariables(Particles.SubImageIndex=&UniformRangedFloat(Minimum=0, Maximum=3)) {isSetParameters=true}
   /SetVariables(Particles.SpriteAlignment=&UniformRangedVector(Minimum=<-0.25 -0.25 1>, Maximum=<0.25 0.25 1>)) {isSetParameters=true}
  ]
  particleUpdateScript [
   &ParticleState()
   &FN_DivingAttributes() {module=FN_DivingAttributes001}
   &SampleSkeletalMeshSkeleton(Mesh:NiagaraDataInterfaceSkeletalMesh=$User.SkeletalMesh, Specific Bone Index:NiagaraInt32=$Particles.SpawnGroup)
   &SkeletalMeshSkeletonLocation(Module.Inherit Velocity=false, Module.Use Sampled Bone as Ribbon ID=true)
   &SolveForcesAndVelocity(Clamp Velocity=false, Limit Acceleration=false) {enabled=false}
   &ScaleColor(
    Scale Alpha=&FloatFromCurve(FloatCurve=DataInterface() {propertyValues {Curve {keys [<at -0.0096 0>,<at 0.5 1>,<at 1.0011 0.002>]}}})
    ScaleA=true
    ScaleRGB=false
   )
   &ScaleSpriteSize(Scale Factor=&Vector2DFromFloat(Value=&Lerp_Float(A=1.5, B=3.5, Alpha=$Particles.Dive)))
  ]
  rendererClasses [&NiagaraSpriteRendererProperties]
  renderers [{rendererClass=&NiagaraSpriteRendererProperties}]
 }
 FN_Contrails_RibbonWde {
  emitterUpdateScript [
   &EmitterScalability(
    Module.bApplySpawnCountScale=false
    Module.bApplySpawnCountScaleByDistanceFraction=false
    Module.bUseMaxDistance=false
    Module.bUseMinDistance=false
   ) {enabled=false}
   &EmitterLifeCycle(CompleteOnInactive=true) {enabled=false}
   &EmitterState()
   /Game/Effects/Niagara/Modules/Spawn/FN_SpawnContrails(All limbs=false, Global Spawnrate=10) {enabled=false}
   &FN_SpawnRibbon_v2(Number of Ribbons=2, Spawn Mode=::NewEnumerator0)
  ]
  particleSpawnScript [
   &FN_InitializeRibbon_v3(Enable Skeletal Position=true, Lifetime=1.25, Ribbon Width=40)
   &AddVelocity(Velocity=<0 0 410>) {enabled=false}
   &AddVelocity(Velocity=&FN_FrontEndVector3Modifier(Gameplay=<0 0 410>, Front End=<0 0 350>)) {module=AddVelocity001}
  ]
  particleUpdateScript [
   &AccelerationForce(
    Acceleration=&VectorFromCurve(VectorCurve=DataInterface() {propertyValues {
       XCurve {keys [<at 0.0017 132.1921>,<at 0.9905 -65.3336>]}
       YCurve {keys [<at -0.0011 1.0067>,<at 0.9988 0>]}
       ZCurve {keys [<at -0.0011 -300>,<at 0.9957 -200.05>]}
      }})
    Coordinate Space=::Local
   ) {enabled=false}
   &Drag(Drag=0.9, Use Linear Drag=true, Use Rotational Drag=false)
   &InheritVelocity(
    Module.Limit Speed=true
    Velocity Limit=10000
    Velocity Scale=<-0.03 -0.03 -0.01>
    Velocity Source=&TransformVector(Vector=$Engine.Owner.Velocity, Source Space=::World, Destination Space=::Local)
   ) {enabled=false}
   &FN_DivingAttributes()
   &SolveForcesAndVelocity(Clamp Velocity=false, Limit Acceleration=false)
   &FN_UpdateRibbon(
    Enable Skeletal Position=true
    Initial Color:LinearColor=$Particles.Initial.Color
    Initial RibbonWidth:NiagaraFloat=$Particles.Initial.RibbonWidth
   )
   &Material(Base Texture Time:NiagaraFloat=&Multiply_Float(A=$Emitter.Age, B=0)) {enabled=false}
   &DynamicMaterialParameters(Module.Param0WriteEnabled=true, Module.Param1WriteEnabled=true, Module.Param2WriteEnabled=true, Module.Param3WriteEnabled=true) {enabled=false}
   &ScaleRibbonWidth(Ribbon Width Scale=&FloatFromCurve(FloatCurve=DataInterface() {propertyValues {Curve {keys [
         {arriveTangent=4.837, interpMode=RCIM_Cubic, leaveTangent=2.478, tangentMode=RCTM_User, time=0.0058, value=0.5}
         {leaveTangent=0.9504, time=0.4538, value=1}
        ]}}}))
   &Color(
    Color=&ColorFromCurve(ColorCurve=DataInterface() {propertyValues {
       AlphaCurve {keys [<at 0.3766 1 RCIM_Cubic>,<at 0.9857 -0.003 RCIM_Cubic>]}
       BlueCurve {keys [<at -0.0023 0 RCIM_Cubic>,<at 0.3035 0>]}
       GreenCurve {keys [<at -0.0023 1 RCIM_Cubic>,<at 0.3035 0.1855>]}
       RedCurve {keys [<at -0.0023 0.1629 RCIM_Cubic>,<at 0.3035 0.7656>]}
      }})
    Scale Alpha=&FloatFromCurve(
     FloatCurve=DataInterface() {propertyValues {Curve {keys [{arriveTangent=0.27, interpMode=RCIM_Cubic, leaveTangent=0.27, tangentMode=RCTM_User, time=-0.0046, value=1.0113}, <at 0.6 0.002>]}}}
     Scale Curve=0.8
    )
   )
   &DynamicMaterialParameters(
    Index 0 Param 1=1
    Index 0 Param 2=1
    Index 0 Param 3=0.1
    Index 0 Param 4=0.2
    Index 1 Param 1=&Add_Float(
     A=&Multiply_Float(A=1, B=&MakeFloatFromInt(Int=HlslExpression() {hlslExpression=Particles.RibbonID}))
     B=&Multiply_Float(A=-1, B=$Engine.Time)
    )
    Index 1 Param 3=1
    Index 1 Param 4=1
    Index 2 Param 1=0.5
    Index 2 Param 2=0.5
    Index 2 Param 3=0.5
    Index 3 Param 1=1
    Index 3 Param 2=1
    Index 3 Param 3=1
    Index 3 Param 4=1
    Module.Param0WriteEnabled=true
    Module.Param0WriteEnabled1=true
    Module.Param0WriteEnabled2=true
    Module.Param0WriteEnabled3=true
    Module.Param1WriteEnabled=true
    Module.Param1WriteEnabled1=true
    Module.Param1WriteEnabled2=true
    Module.Param1WriteEnabled3=true
    Module.Param2WriteEnabled=true
    Module.Param2WriteEnabled1=true
    Module.Param2WriteEnabled2=true
    Module.Param2WriteEnabled3=true
    Module.Param3WriteEnabled=true
    Module.Param3WriteEnabled1=true
    Module.Param3WriteEnabled2=true
    Module.Param3WriteEnabled3=true
    Write Parameter Index 1=true
    Write Parameter Index 2=true
    Write Parameter Index 3=true
   ) {module=DynamicMaterialParameters001}
  ]
  rendererClasses [&NiagaraRibbonRendererProperties]
  renderers [
   {rendererClass=&NiagaraRibbonRendererProperties}
   {rendererClass=&NiagaraRibbonRendererProperties, rendererIndex=1}
  ]
 }
}
summary {
 systemName=NS_Contrail_Informer
 emitters [
  {emitterName=FN_Contrails_LightRays, rendererClasses [&NiagaraSpriteRendererProperties]}
  {emitterName=FN_Contrails_InformerPatterns_WithShade, rendererClasses [&NiagaraSpriteRendererProperties]}
  {emitterName=FN_Contrails_RibbonWde, rendererClasses [&NiagaraRibbonRendererProperties]}
  {emitterName=FN_Contrails_Decal, rendererClasses [&NiagaraSpriteRendererProperties]}
  {
   emitterName=FN_Contrails_Light
   rendererClasses [&NiagaraLightRendererProperties, &NiagaraSpriteRendererProperties]
  }
  {emitterName=FN_Contrails_Fire_Mesh, rendererClasses [&NiagaraMeshRendererProperties]}
  {emitterName=FN_COntrails_Cylinders, rendererClasses [&NiagaraMeshRendererProperties]}
  {emitterName=FN_COntrails_Squares, rendererClasses [&NiagaraSpriteRendererProperties]}
 ]
 userVariables [
  User.bIsFrontEnd {type=NiagaraBool}
  User.bIsFrontEndPreview {type=NiagaraBool}
  User.PlayerSpeed {type=NiagaraFloat, defaultValue=0, description=""}
  User.RibbonIDOffset {type=NiagaraInt32}
  User.SkeletalMesh {
   type=NiagaraDataInterfaceSkeletalMesh
   defaultValue {
    struct=NiagaraExt_VariableValue_DataInterface
    value {dataInterface=&self:NiagaraDataInterfaceSkeletalMesh_0}
   }
  }
 ]
}
